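# Post-copy animation Blueprint for the F_MED_Global_FB female body cosmetic, adding zipper motion and glute correctives to the attached parent's pose.
# A weighted constraint to clavicle, upper spine and neck bones positions the zipper's parent, then spring-driven AnimDynamics swings the zipper.
# Additive glute bone transforms blend in by each thigh's twist relative to the pelvis, computed every update.
# Two zero-alpha thigh rotation nodes sit disconnected from the output.
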AnimBlueprint /BRCosmetics/Characters/Player/Female/Medium/Bodies/F_MED_Global_FB/Meshes/F_MED_Global_FB_AnimBP
target_skeleton=&/F_MED_Global_FB_Skeleton
graphs [
 AnimGraph {nodes [
   n1 = "Transform (Modify) Bone - Bone: thigh_l"(Alpha=0, Translation="0,0,0", Rotation="0,-90,0", Scale="1,1,1") {
    rotationMode=BMM_Additive
    rotationSpace=BCS_BoneSpace
    scaleMode=BMM_Additive
    scaleSpace=BCS_BoneSpace
    translationMode=BMM_Additive
    translationSpace=BCS_BoneSpace
    alphaBoolBlend {blendOption=Linear}
    boneToModify {boneName=thigh_l}
    rotation {yaw=-90}
   }
   "Transform (Modify) Bone - Bone: thigh_r"(ComponentPose=@n1, Alpha=0, Translation="0,0,0", Rotation="0,-90,0", Scale="1,1,1") {
    rotationMode=BMM_Additive
    rotationSpace=BCS_BoneSpace
    scaleMode=BMM_Additive
    scaleSpace=BCS_BoneSpace
    translationMode=BMM_Additive
    translationSpace=BCS_BoneSpace
    alphaBoolBlend {blendOption=Linear}
    boneToModify {boneName=thigh_r}
    rotation {yaw=-90}
   }
   n2 = "Constraint - dyn_parent_zipper (PPPPPP)"!(
    ComponentPose=Local To Component(LocalPose=Copy Pose From Mesh() {bUseAttachedParent=true})
    Alpha=1
    ConstraintWeights_0=0.0285
    ConstraintWeights_1:float=0.0357
    ConstraintWeights_2:float=0.8851
    ConstraintWeights_3:float=0.0283
    ConstraintWeights_4:float=0.0122
    ConstraintWeights_5:float=0.0103
   ) -> (Pose) {
    alphaBoolBlend {blendOption=Linear}
    boneToModify {boneName=dyn_parent_zipper}
    constraintSetup [
     {targetBone {boneName=clavicle_l}}
     {targetBone {boneName=spine_04}}
     {targetBone {boneName=spine_05}}
     {targetBone {boneName=clavicle_r}}
     {targetBone {boneName=neck_02}}
     {targetBone {boneName=neck_01}}
    ]
    constraintWeights [1,1,1,1,1,1]
   }
   n3 = Anim Dynamics - Bone: dyn_zipper(ComponentPose=@n2, Alpha=1) {
    angularDampingOverride=1
    angularSpringConstant=1500
    bAngularSpring=true
    bGravityOverrideInSimSpace=true
    bOverrideAngularDamping=true
    bUseGravityOverride=true
    gravityScale=0.35
    simulationSpace=RootRelative
    alphaBoolBlend {blendOption=Linear}
    boundBone {boneName=dyn_zipper}
    gravityOverride {z=-980}
    physicsBodyDefinitions [{
      sphereCollisionRadius=0
      boundBone {boneName=dyn_zipper}
      boxExtents=<5 5 5>
      constraintSetup {angularLimitsMax=<0 20 0>, angularLimitsMin=<0 -20 -40>, angularTarget=<0 0 0>}
      localJointOffset=<1 0 0>
     }]
    relativeSpaceBone {boneName=dyn_zipper}
   }
   n4 = "Transform (Modify) Bone - Bone: dyn_glute_l"(ComponentPose=@n3, Alpha=Get thigh_rot_l(), Translation="-3,0,-5", Rotation="0.000,0.000,-30", Scale="1.5,1.5,1.5") {
    rotationMode=BMM_Additive
    rotationSpace=BCS_BoneSpace
    scaleMode=BMM_Additive
    scaleSpace=BCS_BoneSpace
    translationMode=BMM_Additive
    translationSpace=BCS_BoneSpace
    alphaBoolBlend {blendOption=Linear}
    boneToModify {boneName=dyn_glute_l}
    rotation {roll=-30}
    scale=<1.5 1.5 1.5>
    translation {x=-3, z=-5}
   }
   n5 = "Transform (Modify) Bone - Bone: dyn_glute_r"(ComponentPose=@n4, Alpha=Get thigh_rot_r(), Translation="-6,0.875,0", Rotation="0,0,-30", Scale="1.5,1.5,1.5") {
    rotationMode=BMM_Additive
    rotationSpace=BCS_BoneSpace
    scaleMode=BMM_Additive
    scaleSpace=BCS_BoneSpace
    translationMode=BMM_Additive
    translationSpace=BCS_BoneSpace
    alphaBoolBlend {blendOption=Linear}
    boneToModify {boneName=dyn_glute_r}
    rotation {roll=-30}
    scale=<1.5 1.5 1.5>
    translation=<-6 0.875>
   }
   Output Pose(Result=Component To Local(ComponentPose=@n5)) {layerGroup=None}
   Get thigh_rot_l()
   Get thigh_rot_r()
  ]}
 EventGraph {nodes [
   Event BlueprintUpdateAnimation()
   > Get!(execute:Exec) -> (then:Exec, else:Exec, Skeletal_Mesh:Skeletal Mesh Component Object Reference)
   > n1 = Sequence()
   > Set thigh_rot_l(thigh_rot_l=GetAlphaFromReferenceTransformDelta(
     SkeletalMeshComponent=Get Skeletal_Mesh() -> (Skeletal_Mesh:Skeletal Mesh Component Object Reference)
     BoneName=thigh_l
     BaseName=pelvis
     DeltaComponentType=QuaternionTwist
     InputRange=(X=20,Y=100)
     TwistAxis="0,0,1"
    ))
   Set thigh_rot_r(
    execute=@n1.then_1
    thigh_rot_r=GetAlphaFromReferenceTransformDelta(
     SkeletalMeshComponent=Get Skeletal_Mesh()
     BoneName=thigh_r
     BaseName=pelvis
     DeltaComponentType=QuaternionTwist
     InputRange=(X=20,Y=100)
     TwistAxis="0,0,1"
    )
   )
   Event BlueprintInitializeAnimation()
   > Set Skeletal_Mesh(Skeletal_Mesh:Skeletal Mesh Component Object Reference=GetOwnerSkeletalMeshComponent()) -> (Output_Get:Skeletal Mesh Component Object Reference)
  ]}
]
variables [
 Skeletal_Mesh {type=SkeletalMeshComponent}
 thigh_rot_l {type=number}
 thigh_rot_r {type=number}
]
